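# Masked parent material for the NilsHolographic vehicle wheel, with a Tread Mat switch selecting rim or tire behavior.
# Paint and metal colors tint cubemap reflections, blended by rim texture masks and fed to base color and emissive.
# Glow combines a panning gradient wave, sine pulse and Fresnel, with tread detail masked from normal-map slopes.
# An optional LOD opacity mask is set before the shared damage and DeRez transition functions.
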Material /VehicleCosmetics/Wheels/NilsHolographic/Materials/M_WheelParentEffect_NilsHolographic
graph {
 properties {
  bAutomaticallySetUsageInEditor=false
  bUseMaterialAttributes=true
  bUsedWithMeshDeformer=true
  bUsedWithSkeletalMesh=true
  blendMode=BLEND_Masked
  previewMesh=/VehicleCosmetics/Wheels/CobaltDagger/Meshes/SK_Wheel_CobaltDagger
  usedShadingModels=Default Lit
 }
 nodes [
  n1 = VectorParameter() {parameterName=PaintColor, defaultValue=<rgba 1 0.94 0 1>}
  n2 = Desaturation(None=@n1.RGB, Fraction=Constant())
  n3 = TextureSample(UVs=Transform(None=ReflectionVectorWS())) {samplerType=SAMPLERTYPE_LinearColor, texture=/VehicleCosmetics/Mutable/FX_Materials/WorldEnv/WorldEnv_AthenaCubemap}
  n4 = Multiply(A=@n3.RGB, B=@n2)
  n5 = StaticSwitchParameter(True=Constant() {r=1}, False=Constant()) {parameterName=Tread Mat}
  n6 = Multiply(A=Multiply(A=@n4, B=@n4), B=@n5)
  n7 = Constant() {r=0.5}
  n8 = Multiply(A=@n1.RGB, B=Constant() {r=5})
  n9 = TextureSampleParameter2D() {
   parameterName=Nils_Rim_RGB
   samplerSource=SSM_Wrap_WorldGroupSettings
   samplerType=SAMPLERTYPE_LinearColor
   texture=/VehicleCosmetics/Wheels/NilsHolographic/Textures/T_NilsHolographic_Rim_RGB
  }
  n10 = TextureCoordinate()
  n11 = ComponentMask(None=@n10) {g=true}
  n12 = ComponentMask(None=@n10) {r=true}
  n13 = TextureSample(UVs=Panner(Coordinate=AppendVector(A=@n11, B=@n12)) {speedX=-0.5}) {samplerSource=SSM_Wrap_WorldGroupSettings, texture=/VehicleCosmetics/Wheels/_SharedTextures/FX/SharpGradient_01_Heightmap}
  n14 = ScalarParameter() {defaultValue=5, parameterName=WaweGlowIntensity}
  n15 = LinearInterpolate(A=Multiply(A=Multiply(A=@n9.R, B=@n13.R), B=@n14), B=Constant(), Alpha=@n5)
  n16 = Multiply(A=Constant() {r=5}, B=ConstantBiasScale(None=@n5))
  n17 = TextureSampleParameter2D() {
   parameterName=NilsRim_N
   samplerSource=SSM_Wrap_WorldGroupSettings
   samplerType=SAMPLERTYPE_Normal
   texture=/VehicleCosmetics/Wheels/NilsHolographic/Textures/T_NilsHolographic_Rim_N
  }
  n18 = TextureSampleParameter2D() {
   parameterName=Tread_N2
   samplerSource=SSM_Wrap_WorldGroupSettings
   samplerType=SAMPLERTYPE_Normal
   texture=/VehicleCosmetics/Wheels/NilsHolographic/Textures/T_NilsHolographic_Tire_N
  }
  n19 = MaterialFunctionCall!(Normal=LinearInterpolate(A=@n17.RGB, B=@n18.RGB, Alpha=@n5), Exponent) {materialFunction=/VehicleCosmetics/Wheels/Cola_Holographic/Materials/MF_Fresnel_Classic}
  n20 = OneMinus(None=Sine(None=Time()) {period=2})
  n21 = Add(A=Multiply(A=Clamp(None=@n20), B=@n14), B=Constant())
  n22 = Multiply(A=@n9.G, B=@n21)
  n23 = Clamp(None=Max(A=Abs(None=@n18.R), B=Abs(None=@n18.G)))
  n24 = Multiply(A=Clamp(None=ConstantBiasScale(None=@n23) {bias=-0.1, scale=5}) {maxDefault=5}, B=@n21)
  n25 = Multiply(
   A=ScalarParameter() {defaultValue=0.5, parameterName=EmissiveIntensity}
   B=Max(A=Multiply(A=@n16, B=@n19.Result), B=LinearInterpolate(A=@n22, B=@n24, Alpha=@n5))
  )
  n26 = Add(
   A=Clamp(None=Multiply(A=@n6, B=@n7))
   B=Multiply(A=LinearInterpolate(A=@n2, B=@n8, Alpha=@n15), B=ConstantBiasScale(None=@n25) {bias=0.02, scale=1})
  )
  n27 = MaterialFunctionCall(Exponent=Constant() {r=1.5}) {materialFunction=&MF_Fresnel_Classic}
  n28 = ConstantBiasScale(None=Multiply(A=VectorParameter() {parameterName=MetallColor, defaultValue=<rgba 0.155 0.51 1 1>}.RGB, B=Multiply(A=@n27.Result, B=@n3.RGB))) {bias=0.05, scale=1}
  n29 = LinearInterpolate(A=@n9.R, B=Constant(), Alpha=@n5)
  n30 = ScalarParameter() {defaultValue=1, parameterName=Roughness}
  n31 = Multiply(A=@n28, B=Constant() {r=0.5})
  n32 = Constant() {r=10}
  n33 = OneMinus(None=@n13.R)
  n34 = Clamp(None=ConstantBiasScale(None=Clamp(None=@n33)) {bias=-0.9, scale=0.25})
  n35 = Multiply(A=@n29, B=Multiply(A=Multiply(A=@n32, B=@n34), B=@n14))
  n36 = TextureSample() {
   samplerSource=SSM_Wrap_WorldGroupSettings
   samplerType=SAMPLERTYPE_LinearColor
   texture=/VehicleCosmetics/Wheels/NilsHolographic/Textures/T_NilsHolographic_Rim_RGB_Alfa_LOD
  }
  n37 = SetMaterialAttributes!(
   MaterialAttributes
   Base Color=LinearInterpolate(A=@n26, B=@n28, Alpha=@n29)
   Metallic=ScalarParameter() {parameterName=Metallic}
   Roughness=@n30
   Emissive Color=Max(A=LinearInterpolate(A=@n26, B=@n31, Alpha=@n29), B=Multiply(A=@n2, B=@n35))
   Opacity Mask=StaticSwitchParameter(True=@n36.R, False=Constant() {r=1}) {parameterName=LODs_Mask}
  ) {attributeSetTypes [
    #1
    #2
    #3
    #4
    #5
   ]}
  MaterialFunctionCall(In=MaterialFunctionCall(Material=@n37) {materialFunction=&MF_BaseVehicle_Damage}.Result) {materialFunction=&MF_VehicleDeRez_Transition}
  Subtract(A=@n30)
 ]
}
